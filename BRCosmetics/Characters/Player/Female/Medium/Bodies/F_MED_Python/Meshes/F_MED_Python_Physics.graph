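# Physics asset for the F_MED_Python female player cosmetic body, covering its two dangling grenade bones.
# Each grenade bone gets a capsule body laid along the bone, and both attach to a small box body on the shared dyn_grenade_parent bone.
# Two constraints with tight swing and even tighter twist limits keep each grenade to a slight wobble around that parent.
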PhysicsAsset /BRCosmetics/Characters/Player/Female/Medium/Bodies/F_MED_Python/Meshes/F_MED_Python_Physics
bodies {
 dyn_grenade_01 [{shapeType=Capsule, length=5.2568, radius=2.8464, center=<3.3891 0 0>, rotation=<rot 90 0 0>}]
 dyn_grenade_02 [{shapeType=Capsule, length=5.2779, radius=2.952, center=<4.6504 0 0>, rotation=<rot 90 0 0>}]
 dyn_grenade_parent [{shapeType=Box, extentX=0.5, extentY=0.5, extentZ=0.5}]
}
constraints [
 {bone1Name=dyn_grenade_02, bone2Name=dyn_grenade_parent, swing1LimitDegrees=10, swing2LimitDegrees=10, twistLimitDegrees=3}
 {bone1Name=dyn_grenade_01, bone2Name=dyn_grenade_parent, swing1LimitDegrees=10, swing2LimitDegrees=10, twistLimitDegrees=3}
]
