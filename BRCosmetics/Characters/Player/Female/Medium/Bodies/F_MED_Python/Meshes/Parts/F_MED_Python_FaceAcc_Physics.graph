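# Physics asset for the face accessory part of the female medium Python player cosmetic, driving secondary motion on a bang and left and right hair bones.
# Capsules cover the dynamic hair bones and upper spine, while small boxes sit on the head and hair base bones.
# Each side chains from a base to a parent bone that branches into two strands, and the bang hangs from the head.
# Constraints set tight twist limits with wider, mostly asymmetric swing limits.
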PhysicsAsset /BRCosmetics/Characters/Player/Female/Medium/Bodies/F_MED_Python/Meshes/Parts/F_MED_Python_FaceAcc_Physics
bodies {
 dyn_bang [{shapeType=Capsule, length=2.42, radius=0.9036, center=<1.3651 -0.2591 0.5706>, rotation=<rot 0 73.8815 -50.0176>}]
 dyn_hair_01_l [{shapeType=Capsule, length=2.0401, radius=1.067, center=<1.7076 0 0>, rotation=<rot 90 0 0>}]
 dyn_hair_01_r [{shapeType=Capsule, length=1.7394, radius=1.0047, center=<1.5541 0 0>, rotation=<rot 90 0 0>}]
 dyn_hair_02_l [{shapeType=Capsule, length=2.2613, radius=1.1313, center=<2.2579 0 0>, rotation=<rot 90 0 0>}]
 dyn_hair_02_r [{shapeType=Capsule, length=1.3803, radius=1.197, center=<1.4226 0 0>, rotation=<rot 90 0 0>}]
 dyn_hair_base_l [{shapeType=Box, extentX=0.5, extentY=0.5, extentZ=0.5}]
 dyn_hair_base_r [{shapeType=Box, extentX=0.5, extentY=0.5, extentZ=0.5}]
 dyn_hair_parent_l [{shapeType=Capsule, length=2.505, radius=2.005, center=<3.287 0 -0.0001>, rotation=<rot 90 0 0>}]
 dyn_hair_parent_r [{shapeType=Capsule, length=2.505, radius=2.005, center=<3.287 0 -0.0001>, rotation=<rot 90 0 0>}]
 head [{shapeType=Box, extentX=2.75, extentY=2.75, extentZ=2.75}]
 spine_05 [{shapeType=Capsule, length=23.625, radius=6.375, center=<7.4603 -3.6655 0>}]
}
constraints [
 {bone1Name=dyn_hair_01_l, bone2Name=dyn_hair_parent_l, swing1LimitDegrees=15, swing2LimitDegrees=30, twistLimitDegrees=3}
 {bone1Name=dyn_hair_02_l, bone2Name=dyn_hair_parent_l, swing1LimitDegrees=15, swing2LimitDegrees=30, twistLimitDegrees=3}
 {bone1Name=dyn_hair_02_r, bone2Name=dyn_hair_parent_r, swing1LimitDegrees=15, swing2LimitDegrees=30, twistLimitDegrees=3}
 {bone1Name=dyn_hair_01_r, bone2Name=dyn_hair_parent_r, swing1LimitDegrees=15, swing2LimitDegrees=30, twistLimitDegrees=3}
 {bone1Name=dyn_bang, bone2Name=head, swing1LimitDegrees=20, swing2LimitDegrees=20, twistLimitDegrees=3}
 {bone1Name=dyn_hair_parent_l, bone2Name=dyn_hair_base_l, swing1LimitDegrees=40, swing2LimitDegrees=10, twistLimitDegrees=3}
 {bone1Name=dyn_hair_parent_r, bone2Name=dyn_hair_base_r, swing1LimitDegrees=40, swing2LimitDegrees=10, twistLimitDegrees=3}
]
